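# Physics asset for the Moss Bomber pickaxe, giving its dangling jaw, flowers, ribbons, mane and branches bodies for secondary motion.
# Each part has a tiny capsule MSTR anchor constrained to the long handle capsule with free swing and twist.
# Most anchors carry a single box body, while the branch anchor drives a five-capsule chain.
# Child joints set per-part angular limits, though many leave swing or twist free.
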PhysicsAsset /BRCosmetics/Weapons/FORT_Melee/Pickaxe_Moss_Bomber/Meshes/Moss_Bomber_Axe_Physics
bodies {
 Melee_TwoHanded_Handle [{shapeType=Capsule, length=87.5284, radius=3.2268, center=<0 0 21.8561>}]
 dyn_MidBranch [{shapeType=Box, extentX=10, extentY=10, extentZ=2, center=<-0.0002 0 -8.5987>}]
 dyn_branch_01 [{shapeType=Capsule, length=3.1712, radius=2, center=<3.4398 -0.0002 0>, rotation=<rot 90 0 0>}]
 dyn_branch_02 [{shapeType=Capsule, length=2, radius=2, center=<2.9578 -0.0002 0>, rotation=<rot 90 0 0>}]
 dyn_branch_03 [{shapeType=Capsule, length=2, radius=2, center=<2.9578 -0.0002 0>, rotation=<rot 90 0 0>}]
 dyn_branch_04 [{shapeType=Capsule, length=2, radius=2, center=<2.9578 -0.0002 0>, rotation=<rot 90 0 0>}]
 dyn_branch_05 [{shapeType=Capsule, length=2, radius=2, center=<2.9578 -0.0002 0>, rotation=<rot 90 0 0>}]
 dyn_branch_MSTR [{shapeType=Capsule, length=0.505, radius=0.505, rotation=<rot 0 -90 0>}]
 dyn_flower_a [{shapeType=Box, extentX=10, extentY=10, extentZ=2, center=<0 0 -6.6076>}]
 dyn_flower_a_MSTR [{shapeType=Capsule, length=0.505, radius=0.505, rotation=<rot 0 -90 0>}]
 dyn_flower_b [{shapeType=Box, extentX=10, extentY=10, extentZ=2, center=<0 0 -5.5033>}]
 dyn_flower_b_MSTR [{shapeType=Capsule, length=0.505, radius=0.505, rotation=<rot 0 -90 0>}]
 dyn_flower_c [{shapeType=Box, extentX=10, extentY=10, extentZ=2, center=<0 0 -5.5033>}]
 dyn_flower_c_MSTR [{shapeType=Capsule, length=0.505, radius=0.505, rotation=<rot 0 -90 0>}]
 dyn_jaw [{shapeType=Box, extentX=10, extentY=10, extentZ=2, center=<2.386 -0.0001 -2.6452>, rotation=<rot 42.4654 0 0>}]
 dyn_jaw_MSTR [{shapeType=Capsule, length=0.505, radius=0.505, rotation=<rot 0 -90 0>}]
 dyn_lowerBranch [{shapeType=Box, extentX=10, extentY=10, extentZ=2, center=<-0.0002 0 -5.8356>}]
 dyn_lowerBranch_MSTR [{shapeType=Capsule, length=0.505, radius=0.505, rotation=<rot 0 -90 0>}]
 dyn_mane_a [{shapeType=Box, extentX=10, extentY=10, extentZ=2, center=<-0.7814 0 -2.2551>}]
 dyn_mane_a_MSTR [{shapeType=Capsule, length=0.505, radius=0.505, rotation=<rot 0 -90 0>}]
 dyn_mane_b [{shapeType=Box, extentX=10, extentY=10, extentZ=2, center=<-0.7814 0 -2.2551>}]
 dyn_mane_b_MSTR [{shapeType=Capsule, length=0.505, radius=0.505, rotation=<rot 0 -90 0>}]
 dyn_mane_c [{shapeType=Box, extentX=10, extentY=10, extentZ=2, center=<-0.0002 0 -2.0346>}]
 dyn_mane_c_MSTR [{shapeType=Capsule, length=0.505, radius=0.505, rotation=<rot 0 -90 0>}]
 dyn_midBranch_MSTR [{shapeType=Capsule, length=0.505, radius=0.505, rotation=<rot 0 -90 0>}]
 dyn_miniBranch_a [{shapeType=Box, extentX=10, extentY=10, extentZ=2, center=<0 0 -4.7574>}]
 dyn_miniBranch_a_MSTR [{shapeType=Capsule, length=0.505, radius=0.505, rotation=<rot 0 -90 0>}]
 dyn_miniBranch_b [{shapeType=Box, extentX=10, extentY=10, extentZ=2, center=<-0.7814 0 -2.2551>}]
 dyn_miniBranch_b_MSTR [{shapeType=Capsule, length=0.505, radius=0.505, rotation=<rot 0 -90 0>}]
 dyn_miniBranch_c [{shapeType=Box, extentX=10, extentY=10, extentZ=2, center=<-0.0002 0 -2.0346>}]
 dyn_miniBranch_c_MSTR [{shapeType=Capsule, length=0.505, radius=0.505, rotation=<rot 0 -90 0>}]
 dyn_ribbon_a [{shapeType=Box, extentX=8.4887, extentY=7.3089, extentZ=2, center=<0.8829 0 -3.8563>}]
 dyn_ribbon_a_MSTR [{shapeType=Capsule, length=0.505, radius=0.505, rotation=<rot 0 -90 0>}]
 dyn_ribbon_b [{shapeType=Box, extentX=8.4887, extentY=7.3089, extentZ=2, center=<0.8829 0 -3.8563>}]
 dyn_ribbon_b_MSTR [{shapeType=Capsule, length=0.505, radius=0.505, rotation=<rot 0 -90 0>}]
 dyn_topBranch [{shapeType=Box, extentX=10, extentY=10, extentZ=2, center=<-0.0002 0 -8.5987>}]
 dyn_topBranch_MSTR [{shapeType=Capsule, length=0.505, radius=0.505, rotation=<rot 0 -90 0>}]
}
constraints [
 {
  bone1Name=dyn_jaw_MSTR
  bone2Name=Melee_TwoHanded_Handle
  swing1LimitDegrees=45
  swing1Motion=Free
  swing2LimitDegrees=45
  swing2Motion=Free
  twistLimitDegrees=45
  twistMotion=Free
 }
 {
  bone1Name=dyn_flower_a_MSTR
  bone2Name=Melee_TwoHanded_Handle
  swing1LimitDegrees=45
  swing1Motion=Free
  swing2LimitDegrees=45
  swing2Motion=Free
  twistLimitDegrees=45
  twistMotion=Free
 }
 {
  bone1Name=dyn_flower_b_MSTR
  bone2Name=Melee_TwoHanded_Handle
  swing1LimitDegrees=45
  swing1Motion=Free
  swing2LimitDegrees=45
  swing2Motion=Free
  twistLimitDegrees=45
  twistMotion=Free
 }
 {
  bone1Name=dyn_flower_c_MSTR
  bone2Name=Melee_TwoHanded_Handle
  swing1LimitDegrees=45
  swing1Motion=Free
  swing2LimitDegrees=45
  swing2Motion=Free
  twistLimitDegrees=45
  twistMotion=Free
 }
 {
  bone1Name=dyn_branch_MSTR
  bone2Name=Melee_TwoHanded_Handle
  swing1LimitDegrees=45
  swing1Motion=Free
  swing2LimitDegrees=45
  swing2Motion=Free
  twistLimitDegrees=45
  twistMotion=Free
 }
 {
  bone1Name=dyn_ribbon_a_MSTR
  bone2Name=Melee_TwoHanded_Handle
  swing1LimitDegrees=45
  swing1Motion=Free
  swing2LimitDegrees=45
  swing2Motion=Free
  twistLimitDegrees=45
  twistMotion=Free
 }
 {
  bone1Name=dyn_ribbon_b_MSTR
  bone2Name=Melee_TwoHanded_Handle
  swing1LimitDegrees=45
  swing1Motion=Free
  swing2LimitDegrees=45
  swing2Motion=Free
  twistLimitDegrees=45
  twistMotion=Free
 }
 {
  bone1Name=dyn_topBranch_MSTR
  bone2Name=Melee_TwoHanded_Handle
  swing1LimitDegrees=45
  swing1Motion=Free
  swing2LimitDegrees=45
  swing2Motion=Free
  twistLimitDegrees=45
  twistMotion=Free
 }
 {
  bone1Name=dyn_midBranch_MSTR
  bone2Name=Melee_TwoHanded_Handle
  swing1LimitDegrees=45
  swing1Motion=Free
  swing2LimitDegrees=45
  swing2Motion=Free
  twistLimitDegrees=45
  twistMotion=Free
 }
 {
  bone1Name=dyn_lowerBranch_MSTR
  bone2Name=Melee_TwoHanded_Handle
  swing1LimitDegrees=45
  swing1Motion=Free
  swing2LimitDegrees=45
  swing2Motion=Free
  twistLimitDegrees=45
  twistMotion=Free
 }
 {
  bone1Name=dyn_miniBranch_a_MSTR
  bone2Name=Melee_TwoHanded_Handle
  swing1LimitDegrees=45
  swing1Motion=Free
  swing2LimitDegrees=45
  swing2Motion=Free
  twistLimitDegrees=45
  twistMotion=Free
 }
 {
  bone1Name=dyn_miniBranch_b_MSTR
  bone2Name=Melee_TwoHanded_Handle
  swing1LimitDegrees=45
  swing1Motion=Free
  swing2LimitDegrees=45
  swing2Motion=Free
  twistLimitDegrees=45
  twistMotion=Free
 }
 {
  bone1Name=dyn_miniBranch_c_MSTR
  bone2Name=Melee_TwoHanded_Handle
  swing1LimitDegrees=45
  swing1Motion=Free
  swing2LimitDegrees=45
  swing2Motion=Free
  twistLimitDegrees=45
  twistMotion=Free
 }
 {
  bone1Name=dyn_mane_a_MSTR
  bone2Name=Melee_TwoHanded_Handle
  swing1LimitDegrees=45
  swing1Motion=Free
  swing2LimitDegrees=45
  swing2Motion=Free
  twistLimitDegrees=45
  twistMotion=Free
 }
 {
  bone1Name=dyn_mane_b_MSTR
  bone2Name=Melee_TwoHanded_Handle
  swing1LimitDegrees=45
  swing1Motion=Free
  swing2LimitDegrees=45
  swing2Motion=Free
  twistLimitDegrees=45
  twistMotion=Free
 }
 {
  bone1Name=dyn_mane_c_MSTR
  bone2Name=Melee_TwoHanded_Handle
  swing1LimitDegrees=45
  swing1Motion=Free
  swing2LimitDegrees=45
  swing2Motion=Free
  twistLimitDegrees=45
  twistMotion=Free
 }
 {bone1Name=dyn_jaw, bone2Name=dyn_jaw_MSTR, swing1LimitDegrees=2, swing2LimitDegrees=50, twistLimitDegrees=2}
 {
  bone1Name=dyn_flower_a
  bone2Name=dyn_flower_a_MSTR
  swing1LimitDegrees=30
  swing1Motion=Free
  swing2LimitDegrees=30
  swing2Motion=Free
  twistLimitDegrees=5
 }
 {
  bone1Name=dyn_flower_b
  bone2Name=dyn_flower_b_MSTR
  swing1LimitDegrees=30
  swing1Motion=Free
  swing2LimitDegrees=30
  swing2Motion=Free
  twistLimitDegrees=5
 }
 {
  bone1Name=dyn_flower_c
  bone2Name=dyn_flower_c_MSTR
  swing1LimitDegrees=30
  swing1Motion=Free
  swing2LimitDegrees=30
  swing2Motion=Free
  twistLimitDegrees=5
 }
 {bone1Name=dyn_branch_02, bone2Name=dyn_branch_01, swing1LimitDegrees=30, swing2LimitDegrees=30, twistLimitDegrees=5}
 {bone1Name=dyn_branch_03, bone2Name=dyn_branch_02, swing1LimitDegrees=30, swing2LimitDegrees=30, twistLimitDegrees=5}
 {bone1Name=dyn_branch_04, bone2Name=dyn_branch_03, swing1LimitDegrees=30, swing2LimitDegrees=30, twistLimitDegrees=5}
 {bone1Name=dyn_branch_05, bone2Name=dyn_branch_04, swing1LimitDegrees=30, swing2LimitDegrees=30, twistLimitDegrees=5}
 {
  bone1Name=dyn_topBranch
  bone2Name=dyn_topBranch_MSTR
  swing1LimitDegrees=30
  swing1Motion=Free
  swing2LimitDegrees=30
  swing2Motion=Free
  twistLimitDegrees=5
  twistMotion=Free
 }
 {
  bone1Name=dyn_MidBranch
  bone2Name=dyn_midBranch_MSTR
  swing1LimitDegrees=30
  swing1Motion=Free
  swing2LimitDegrees=30
  swing2Motion=Free
  twistLimitDegrees=5
  twistMotion=Free
 }
 {
  bone1Name=dyn_lowerBranch
  bone2Name=dyn_lowerBranch_MSTR
  swing1LimitDegrees=30
  swing1Motion=Free
  swing2LimitDegrees=30
  swing2Motion=Free
  twistLimitDegrees=5
  twistMotion=Free
 }
 {
  bone1Name=dyn_miniBranch_a
  bone2Name=dyn_miniBranch_a_MSTR
  swing1LimitDegrees=30
  swing1Motion=Free
  swing2LimitDegrees=30
  swing2Motion=Free
  twistLimitDegrees=5
 }
 {
  bone1Name=dyn_miniBranch_b
  bone2Name=dyn_miniBranch_b_MSTR
  swing1LimitDegrees=60
  swing1Motion=Free
  swing2LimitDegrees=60
  swing2Motion=Free
  twistLimitDegrees=15
  twistMotion=Free
 }
 {bone1Name=dyn_miniBranch_c, bone2Name=dyn_miniBranch_c_MSTR, swing1LimitDegrees=30, swing2LimitDegrees=30, twistLimitDegrees=5}
 {
  bone1Name=dyn_mane_a
  bone2Name=dyn_mane_a_MSTR
  swing1LimitDegrees=60
  swing1Motion=Free
  swing2LimitDegrees=60
  swing2Motion=Free
  twistLimitDegrees=15
  twistMotion=Free
 }
 {
  bone1Name=dyn_mane_b
  bone2Name=dyn_mane_b_MSTR
  swing1LimitDegrees=60
  swing1Motion=Free
  swing2LimitDegrees=60
  swing2Motion=Free
  twistLimitDegrees=15
  twistMotion=Free
 }
 {bone1Name=dyn_mane_c, bone2Name=dyn_mane_c_MSTR, swing1LimitDegrees=30, swing2LimitDegrees=30, twistLimitDegrees=5}
 {bone1Name=dyn_branch_01, bone2Name=dyn_branch_MSTR, swing1LimitDegrees=15, swing2LimitDegrees=15, twistLimitDegrees=5}
 {
  bone1Name=dyn_ribbon_a
  bone2Name=dyn_ribbon_a_MSTR
  swing1LimitDegrees=45
  swing1Motion=Free
  swing2LimitDegrees=45
  swing2Motion=Free
  twistLimitDegrees=10
  twistMotion=Free
 }
 {
  bone1Name=dyn_ribbon_b
  bone2Name=dyn_ribbon_b_MSTR
  swing1LimitDegrees=45
  swing1Motion=Free
  swing2LimitDegrees=45
  swing2Motion=Free
  twistLimitDegrees=10
  twistMotion=Free
 }
]
