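# Anim Blueprint for the Moss Bomber pickaxe that adds secondary physics motion to its dynamic branch bones.
# A RigidBody node simulates the input pose in root-bone space with custom gravity, at a lower blend in the front end.
# Five chained constraints then partially pull the branch bones toward the two-handed handle, with weights tapering along the chain.
# A final additive bone transform offsets the branch master bone; the event graph holds no logic.
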AnimBlueprint /BRCosmetics/Weapons/FORT_Melee/Pickaxe_Moss_Bomber/Meshes/Moss_Bomber_Axe_AnimBP
target_skeleton=&/Moss_Bomber_Axe_Skeleton
graphs [
 AnimGraph {nodes [
   n1 = RigidBody(ComponentPose=Local To Component(LocalPose=Input Pose()), Alpha=SelectFloat(A=0.6, B=0.75, bPickA=Get bIsInFrontEnd())) {
    bClampLinearTranslationLimitToRefPose=true
    bOverrideWorldGravity=true
    overridePhysicsAsset=&/Moss_Bomber_Axe_Physics
    simulationSpace=BaseBoneSpace
    alphaBoolBlend {blendOption=Linear}
    baseBoneRef {boneName=b_AR_Root}
    componentAppliedLinearAccClamp=<1500 1500 1500>
    componentLinearAccScale=<0.4 0.4 0.42>
    overrideWorldGravity {z=-980}
   }
   n2 = "Constraint - dyn_branch_01 (P)"(ComponentPose=@n1, Alpha=0.3, ConstraintWeights_0=1) {
    alphaBoolBlend {blendOption=Linear}
    boneToModify {boneName=dyn_branch_01}
    constraintSetup [{targetBone {boneName=Melee_TwoHanded_Handle}}]
    constraintWeights [1]
   }
   n3 = "Constraint - dyn_branch_02 (P)"(ComponentPose=@n2, Alpha=0.3, ConstraintWeights_0=1) {
    alphaBoolBlend {blendOption=Linear}
    boneToModify {boneName=dyn_branch_02}
    constraintSetup [{targetBone {boneName=Melee_TwoHanded_Handle}}]
    constraintWeights [1]
   }
   n4 = "Constraint - dyn_branch_03 (P)"(ComponentPose=@n3, Alpha=0.2, ConstraintWeights_0=1) {
    alphaBoolBlend {blendOption=Linear}
    boneToModify {boneName=dyn_branch_03}
    constraintSetup [{targetBone {boneName=Melee_TwoHanded_Handle}}]
    constraintWeights [1]
   }
   n5 = "Constraint - dyn_branch_04 (P)"(ComponentPose=@n4, Alpha=0.1, ConstraintWeights_0=1) {
    alphaBoolBlend {blendOption=Linear}
    boneToModify {boneName=dyn_branch_04}
    constraintSetup [{targetBone {boneName=Melee_TwoHanded_Handle}}]
    constraintWeights [1]
   }
   n6 = "Constraint - dyn_branch_05 (P)"(ComponentPose=@n5, Alpha=0.1, ConstraintWeights_0=1) {
    alphaBoolBlend {blendOption=Linear}
    boneToModify {boneName=dyn_branch_05}
    constraintSetup [{targetBone {boneName=Melee_TwoHanded_Handle}}]
    constraintWeights [1]
   }
   n7 = "Transform (Modify) Bone - Bone: dyn_branch_MSTR"(ComponentPose=@n6, Alpha=1, Translation="0,0,0", Rotation="32.967,-2.420,-9.249", Scale="1,1,1") {
    rotationMode=BMM_Additive
    rotationSpace=BCS_BoneSpace
    scaleMode=BMM_Additive
    scaleSpace=BCS_BoneSpace
    translationMode=BMM_Additive
    translationSpace=BCS_BoneSpace
    alphaBoolBlend {blendOption=Linear}
    boneToModify {boneName=dyn_branch_MSTR}
    rotation=<rot 32.967 -2.42 -9.249>
   }
   Output Pose(Result=Component To Local(ComponentPose=@n7)) {layerGroup=None}
  ]}
 EventGraph {nodes [
   TryGetPawnOwner()
   Event BlueprintUpdateAnimation()
  ]}
]
variables []
